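; Dynamic UI director that shows or hides the hoverboard HUD scene as the local player's pawn changes vehicle.
; On BeginPlay outside the Campaign subgame, it binds to the HUD context's pawn-set event and the controlled pawn's vehicle-state event, rebinding the latter whenever a pawn is set.
; Casts to Hoverboard_V2 add the scene for a new hoverboard and remove it when leaving one.
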(event EventBeginPlay
  (if (not (FortGlobals|IsinSubGame "Campaign"))
    (HUD|BindEventtoOnPawnSet (BlueprintContext|GetFortHUDContext) (delegate Custom|OnPawnSet))
    (bind _asfort_player_pawn (Utilities|Casting|CastToFortPlayerPawn (Pawn|GetControlledPawn (DynamicUIDirector|GetOwningLocalPlayerController))))
    (Vehicle|BindEventtoOnVehicleStateChangedEvent _asfort_player_pawn (delegate Custom|UpdateUIScene))))

(event Custom|OnPawnSet
  (bind _asfort_player_pawn (Utilities|Casting|CastToFortPlayerPawn (Pawn|GetControlledPawn (DynamicUIDirector|GetOwningLocalPlayerController))))
  (Vehicle|BindEventtoOnVehicleStateChangedEvent _asfort_player_pawn (delegate Custom|UpdateUIScene)))

(event Custom|UpdateUIScene
  (params (FortPlayerPawn :type Fort%20Player%20Pawn%20Object%20Reference) (NewVehicle :type Actor%20Object%20Reference) (OldVehicle :type Actor%20Object%20Reference))
  (Utilities|Casting|CastToHoverboard_V2 NewVehicle
    (:then
      (DynamicUIDirector|AddScene "/HoverboardV2/HUD/HoverboardUIScene.HoverboardUIScene"))
    (:CastFailed
      (Utilities|Casting|CastToHoverboard_V2 OldVehicle)
      (DynamicUIDirector|RemoveScene "/HoverboardV2/HUD/HoverboardUIScene.HoverboardUIScene"))))

(fn ConstructionScript)
