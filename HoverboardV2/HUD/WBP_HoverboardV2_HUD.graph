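# Fortnite vehicle dashboard widget for the Hoverboard V2, showing an energy bar, boost effects, unlimited-fuel icon and low-energy, empty and disabled warnings.
# SetupVehicle finds the player's board and binds its fuel, boost and disabled delegates.
# Cached fuel drives the bar material and lerps boost spark position, glow scale and fade, while state changes trigger warning and boost animations.
# Mobile touch swaps layout and warning animations, and also listens for vehicle changes.
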WidgetBlueprint /HoverboardV2/HUD/WBP_HoverboardV2_HUD
animations [
 Anim_BoostOn {
  binding_count=3
  length_seconds=0.2
  object=&self:Anim_BoostOn
  sequence {bindings [
    Image_FillBar {
     id=#1
     possessed_class=Image
     tracks [{
       type=MovieSceneWidgetMaterialTrack
       sections [{
         type=MovieSceneParameterSection
         channels [
          Progress_Color_StopR {type=MovieSceneScriptingFloatChannel, const [1,2]}
          Progress_Color_StopG {type=MovieSceneScriptingFloatChannel, const [1,2]}
          Progress_Color_StopB {type=MovieSceneScriptingFloatChannel, const [1,2]}
          Progress_Color_StopA {type=MovieSceneScriptingFloatChannel, const [1,2]}
         ]
        }]
      }]
    }
    Border_BoostGlow {
     id=#2
     possessed_class=Border
     tracks [{
       type=MovieSceneFloatTrack
       display_name=Render Opacity
       sections [{type=MovieSceneFloatSection, channels [None {type=MovieSceneScriptingFloatChannel, keys [[0,0],[4,1]]}]}]
      }]
    }
    Overlay_BoostSparks {
     id=#3
     possessed_class=Overlay
     tracks [{
       type=MovieSceneFloatTrack
       display_name=Render Opacity
       sections [{type=MovieSceneFloatSection, channels [None {type=MovieSceneScriptingFloatChannel, keys [[0,0],[4,1]]}]}]
      }]
    }
   ]}
 }
 Anim_BoostOff {
  binding_count=3
  length_seconds=0.75
  object=&self:Anim_BoostOff
  sequence {bindings [
    Overlay_BoostSparks {
     id=#4
     possessed_class=Overlay
     tracks [{
       type=MovieSceneFloatTrack
       display_name=Render Opacity
       sections [{type=MovieSceneFloatSection, channels [None {type=MovieSceneScriptingFloatChannel, keys [[0,1],[14.9997,0]]}]}]
      }]
    }
    Image_FillBar {
     id=#1
     possessed_class=Image
     tracks [{
       type=MovieSceneWidgetMaterialTrack
       sections [{
         type=MovieSceneParameterSection
         channels [
          Progress_Color_StopR {type=MovieSceneScriptingFloatChannel, const [1,2]}
          Progress_Color_StopG {type=MovieSceneScriptingFloatChannel, const [1,2]}
          Progress_Color_StopB {type=MovieSceneScriptingFloatChannel, const [1,2]}
          Progress_Color_StopA {type=MovieSceneScriptingFloatChannel, const [1,2]}
         ]
        }]
      }]
    }
    Border_BoostGlow {
     id=#5
     possessed_class=Border
     tracks [{
       type=MovieSceneFloatTrack
       display_name=Render Opacity
       sections [{type=MovieSceneFloatSection, channels [None {type=MovieSceneScriptingFloatChannel, keys [[0,1],[15,0]]}]}]
      }]
    }
   ]}
 }
 Anim_LowFuel_Messaging {
  binding_count=6
  length_seconds=0.25
  object=&self:Anim_LowFuel_Messaging
  sequence {
   bindings [
    Image_TriangleArrow {
     id=#6
     possessed_class=Image
     tracks [
      {
       type=MovieSceneFloatTrack
       display_name=Render Opacity
       sections [{type=MovieSceneFloatSection, channels [None {type=MovieSceneScriptingFloatChannel, keys [[0.03,0],[3,1]]}]}]
      }
      {
       type=MovieScene2DTransformTrack
       sections [{
         type=MovieScene2DTransformSection
         channels [
          Translation.X {type=MovieSceneScriptingFloatChannel, keys [[-0.001,16],[5.999,-12],[8.999,-8]]}
          Translation.Y {type=MovieSceneScriptingFloatChannel, keys [[-0.001,0],[5.999,30],[8.999,24]]}
          Angle {type=MovieSceneScriptingFloatChannel, const [108,1]}
          Scale.X {type=MovieSceneScriptingFloatChannel, const [1,1]}
          Scale.Y {type=MovieSceneScriptingFloatChannel, const [1,1]}
          Shear.X {type=MovieSceneScriptingFloatChannel, const [0,1]}
          Shear.Y {type=MovieSceneScriptingFloatChannel, const [0,1]}
         ]
        }]
      }
     ]
    }
    Image_NoEnergy_GlowPulse {
     id=#7
     possessed_class=Image
     tracks [{
       type=MovieSceneFloatTrack
       display_name=Render Opacity
       sections [{type=MovieSceneFloatSection, channels [None {type=MovieSceneScriptingFloatChannel, keys [[0,0],[12,0.5]]}]}]
      }]
    }
    Image_AdditonalOuterRing {
     id=#8
     possessed_class=Image
     tracks [
      {
       type=MovieSceneFloatTrack
       display_name=Render Opacity
       sections [{type=MovieSceneFloatSection, channels [None {type=MovieSceneScriptingFloatChannel, keys [[9,0],[12.001,0.25]]}]}]
      }
      {
       type=MovieScene2DTransformTrack
       sections [{
         type=MovieScene2DTransformSection
         channels [
          Scale.X {type=MovieSceneScriptingFloatChannel, keys [[9,2],[15,1]]}
          Scale.Y {type=MovieSceneScriptingFloatChannel, keys [[9,2],[15,1]]}
         ]
        }]
      }
     ]
    }
    Overlay_WarningMessaging {
     id=#9
     possessed_class=Overlay
     tracks [{
       type=MovieSceneFloatTrack
       display_name=Render Opacity
       sections [{type=MovieSceneFloatSection, channels [None {type=MovieSceneScriptingFloatChannel, keys [[0,0],[3,1]]}]}]
      }]
    }
    Overlay_TextContainerLow {
     id=#10
     possessed_class=Overlay
     tracks [
      {
       type=MovieSceneFloatTrack
       display_name=Render Opacity
       sections [{type=MovieSceneFloatSection, channels [None {type=MovieSceneScriptingFloatChannel, keys [[6,0],[9,1]]}]}]
      }
      {
       type=MovieScene2DTransformTrack
       sections [{
         type=MovieScene2DTransformSection
         channels [
          Translation.X {type=MovieSceneScriptingFloatChannel, keys [[6,20],[9,0],[12.001,8]]}
          Translation.Y {type=MovieSceneScriptingFloatChannel, keys [[6,-16],[9,0],[12.001,-8]]}
          Scale.X {type=MovieSceneScriptingFloatChannel, keys [[6,1.125],[9,0.95],[12.001,1]]}
          Scale.Y {type=MovieSceneScriptingFloatChannel, keys [[6,1.125],[9,0.95],[12.001,1]]}
         ]
        }]
      }
     ]
    }
    Image_Energy {
     id=#11
     possessed_class=Image
     tracks [{
       type=MovieScene2DTransformTrack
       sections [{
         type=MovieScene2DTransformSection
         channels [
          Translation.X {type=MovieSceneScriptingFloatChannel, const [0,1]}
          Translation.Y {type=MovieSceneScriptingFloatChannel, keys [[0,-1],[5.999,-1],[9,-4],[12,-1]]}
          Scale.X {type=MovieSceneScriptingFloatChannel, keys [[0,1],[5.999,1],[9,1.125],[12,1]]}
          Scale.Y {type=MovieSceneScriptingFloatChannel, keys [[0,1],[5.999,1],[9,1.125],[12,1]]}
         ]
        }]
      }]
    }
   ]
   display_rate {numerator=60}
  }
 }
 Anim_SelfDestructMessaging {
  binding_count=5
  length_seconds=0.35
  object=&self:Anim_SelfDestructMessaging
  sequence {
   bindings [
    Overlay_TextContainerEject {
     id=#12
     possessed_class=Overlay
     tracks [
      {
       type=MovieSceneFloatTrack
       display_name=Render Opacity
       sections [{type=MovieSceneFloatSection, channels [None {type=MovieSceneScriptingFloatChannel, keys [[6,0],[15,1]]}]}]
      }
      {
       type=MovieScene2DTransformTrack
       sections [{
         type=MovieScene2DTransformSection
         channels [
          Translation.X {type=MovieSceneScriptingFloatChannel, keys [[6,256],[15,8]]}
          Translation.Y {type=MovieSceneScriptingFloatChannel, const [-8,1]}
          Scale.X {type=MovieSceneScriptingFloatChannel, const [1,1]}
          Scale.Y {type=MovieSceneScriptingFloatChannel, const [1,1]}
         ]
        }]
      }
     ]
    }
    Overlay_AlertIcon {
     id=#13
     possessed_class=Overlay
     tracks [
      {
       type=MovieSceneFloatTrack
       display_name=Render Opacity
       sections [{type=MovieSceneFloatSection, channels [None {type=MovieSceneScriptingFloatChannel, keys [[0,0],[12,0],[15,1]]}]}]
      }
      {
       type=MovieScene2DTransformTrack
       sections [{
         type=MovieScene2DTransformSection
         channels [
          Translation.X {type=MovieSceneScriptingFloatChannel, const [-32,1]}
          Translation.Y {type=MovieSceneScriptingFloatChannel, const [-40,1]}
          Scale.X {type=MovieSceneScriptingFloatChannel, keys [[12,3],[15,0.9],[21,1]]}
          Scale.Y {type=MovieSceneScriptingFloatChannel, keys [[12,3],[15,0.9],[21,1]]}
         ]
        }]
      }
     ]
    }
    Image_AdditonalOuterRing_1 {id=#14, possessed_class=Image, tracks []}
    Overlay_TextContainerLow {
     id=#10
     possessed_class=Overlay
     tracks [
      {
       type=MovieSceneFloatTrack
       display_name=Render Opacity
       sections [{type=MovieSceneFloatSection, channels [None {type=MovieSceneScriptingFloatChannel, keys [[0,1],[9,0]]}]}]
      }
      {
       type=MovieScene2DTransformTrack
       sections [{
         type=MovieScene2DTransformSection
         channels [
          Translation.X {type=MovieSceneScriptingFloatChannel, keys [[0,8],[9,-256]]}
          Translation.Y {type=MovieSceneScriptingFloatChannel, const [-8,1]}
          Scale.X {type=MovieSceneScriptingFloatChannel, const [1,1]}
          Scale.Y {type=MovieSceneScriptingFloatChannel, const [1,1]}
         ]
        }]
      }
     ]
    }
   ]
   display_rate {numerator=60}
  }
 }
 Anim_LowFuel_Messaging_Touch {
  binding_count=3
  length_seconds=0.5014
  object=&self:Anim_LowFuel_Messaging_Touch
  sequence {
   bindings [
    bg_low {
     id=#15
     possessed_class=Image
     tracks [{
       type=MovieSceneColorTrack
       display_name=Color and Opacity
       sections [{
         type=MovieSceneColorSection
         end_frame=30
         start_frame=-1
         channels [
          Color.R {type=MovieSceneScriptingFloatChannel, const [1,4]}
          Color.G {type=MovieSceneScriptingFloatChannel, keys [[0,0.3185],[14.302,0.3185],[15,0.9],[30,0.3185]]}
          Color.B {type=MovieSceneScriptingFloatChannel, const [0,4]}
          Color.A {type=MovieSceneScriptingFloatChannel, const [1,4]}
         ]
        }]
      }]
    }
    OV_touch_lowwarning {
     id=#16
     possessed_class=Overlay
     tracks [{
       type=MovieSceneFloatTrack
       display_name=Render Opacity
       sections [{type=MovieSceneFloatSection, end_frame=29, start_frame=-1, channels [None {type=MovieSceneScriptingFloatChannel, const [1,1]}]}]
      }]
    }
    bg_low_skew {
     id=#17
     possessed_class=Image
     tracks [{
       type=MovieSceneColorTrack
       display_name=Color and Opacity
       sections [{
         type=MovieSceneColorSection
         end_frame=30
         start_frame=-1
         channels [
          Color.R {type=MovieSceneScriptingFloatChannel, const [1,4]}
          Color.G {type=MovieSceneScriptingFloatChannel, keys [[0,0.3185],[14.302,0.3185],[15,0.9],[30,0.3185]]}
          Color.B {type=MovieSceneScriptingFloatChannel, const [0,4]}
          Color.A {type=MovieSceneScriptingFloatChannel, const [1,4]}
         ]
        }]
      }]
    }
   ]
   display_rate {numerator=60}
  }
 }
 Anim_SelfDestructMessaging_Touch {
  binding_count=3
  length_seconds=0.5014
  object=&self:Anim_SelfDestructMessaging_Touch
  sequence {
   bindings [
    bg_empty {
     id=#18
     possessed_class=Image
     tracks [{
       type=MovieSceneColorTrack
       display_name=Color and Opacity
       sections [{
         type=MovieSceneColorSection
         end_frame=30
         start_frame=-1
         channels [
          Color.R {type=MovieSceneScriptingFloatChannel, const [1,4]}
          Color.G {type=MovieSceneScriptingFloatChannel, thinned_from=5, keys [[-0.016,0.0065],[14.302,0.0065],[15.024,0.2],[30,0.0065]]}
          Color.B {type=MovieSceneScriptingFloatChannel, keys [[0,0],[14.302,0],[15.024,0.1],[30,0]]}
          Color.A {type=MovieSceneScriptingFloatChannel, const [1,4]}
         ]
        }]
      }]
    }
    OV_touch_emptywarning {
     id=#19
     possessed_class=Overlay
     tracks [{
       type=MovieSceneFloatTrack
       display_name=Render Opacity
       sections [{type=MovieSceneFloatSection, end_frame=29, start_frame=-1, channels [None {type=MovieSceneScriptingFloatChannel, const [1,4]}]}]
      }]
    }
    bg_empty_skew {
     id=#20
     possessed_class=Image
     tracks [{
       type=MovieSceneColorTrack
       display_name=Color and Opacity
       sections [{
         type=MovieSceneColorSection
         end_frame=30
         start_frame=-1
         channels [
          Color.R {type=MovieSceneScriptingFloatChannel, const [1,4]}
          Color.G {type=MovieSceneScriptingFloatChannel, thinned_from=5, keys [[-0.016,0.0065],[14.302,0.0065],[15.024,0.2],[30,0.0065]]}
          Color.B {type=MovieSceneScriptingFloatChannel, keys [[0,0],[14.302,0],[15.024,0.1],[30,0]]}
          Color.A {type=MovieSceneScriptingFloatChannel, const [1,4]}
         ]
        }]
      }]
    }
   ]
   display_rate {numerator=60}
  }
 }
]
graphs [
 DisabledChange {nodes [
   n1 = DisabledChange!() -> (then, bIsDisabled:bool=~) {is=title K2Node_FunctionEntry}
   > n2 = Branch(Condition=@n1)
   n3 = Get WBP_Vehicle_Warning() -> (WBP_Vehicle_Warning:WBP Vehicle Warning Object Reference)
   > ShowWarning!(execute, self:WBP Vehicle Warning Object Reference=@n3, Type:ENUM_WarningType Enum=NewEnumerator2) -> (then)
   > SetVisibility(self=Get Overlay_WarningMessaging() -> (Overlay_WarningMessaging:Overlay Object Reference), InVisibility=Collapsed)
   RemoveWarning!(execute=@n2.else, self:WBP Vehicle Warning Object Reference=@n3, Type:ENUM_WarningType Enum=NewEnumerator2) -> (then)
   > UpdateLowFuelState()
   > UpdateEmptyState()
  ]}
 EventGraph {nodes [
   UpdateLowFuelState()
   > UpdateEmptyState()
   > SetBoostFXCollapsed()
   Event Update On Boost Finished() {is=title K2Node_CustomEvent}
   > SetBoostState!(execute, self, bIsBoosting:bool=false) -> (then)
   "Event Update On Low Fuel "() {is=title K2Node_CustomEvent}
   > UpdateLowFuelState()
   Event Update On Fuel Value Changed() {is=title K2Node_CustomEvent}
   n1 = Get REF_HoverboardVehicle() -> (REF_HoverboardVehicle:Hoverboard V2 Object Reference)
   n2 = GetVehicleFuelPercent!(self:Fort Vehicle Interface Interface=@n1) -> (ReturnValue:float)
   > Branch(Condition="Not Equal (Float)"(A=@n2, B=Get CachedFuel() -> (CachedFuel:double)))
   > n3 = Set CachedFuel(CachedFuel:double=@n2) -> (Output_Get:double)
   > Set IsEmpty(IsEmpty:bool="Nearly Equal (Float)"(A=@n3, B=0, ErrorTolerance=0)) -> (Output_Get:bool)
   n4 = IsVehicleFuelLow!(self:Fort Vehicle Interface Interface=@n1) -> (ReturnValue:bool)
   > n5 = Branch(Condition="Equal (Boolean)"(A:bool=@n4, B:bool=Get IsLowFuel() -> (IsLowFuel:bool)) -> (ReturnValue:bool=~) {is=title K2Node_PromotableOperator})
   n6 = GetDynamicMaterial(execute=[@n14.then, @n5.then], self=Get Image_FillBar() -> (Image_FillBar:Image Object Reference))
   > SetScalarParameterValue(self=@n6, ParameterName=Progress, Value=Get CachedFuel())
   > UpdateEmptyState()
   n7 = Get BoostSparksEndTranslation() -> (BoostSparksEndTranslation:double)
   n8 = Get BoostSparksStartTranslation() -> (BoostSparksStartTranslation:double)
   n9 = Get CachedFuel()
   > SetRenderTranslation(
    self=Get Overlay_BoostSparks() -> (Overlay_BoostSparks:Overlay Object Reference)
    Translation=MakeVector2D(X=Lerp(A=@n7, B=@n8, Alpha=@n9), Y=0)
   )
   n10 = Get CachedFuel()
   n11 = Get BoostGlowBorderScaleEnd() -> (BoostGlowBorderScaleEnd:double)
   n12 = Get BoostGlowBorderScaleStart() -> (BoostGlowBorderScaleStart:double)
   > SetDesiredSizeScale!(
    execute
    self:Border Object Reference=Get Border_BoostGlow() -> (Border_BoostGlow:Border Object Reference)
    InScale:Vector 2D Structure=MakeVector2D(X=Lerp(A=@n11, B=@n12, Alpha=@n10), Y=0)
   ) -> (then)
   > n13 = GetDynamicMaterial(self=Get Image_BoostGlow() -> (Image_BoostGlow:Image Object Reference))
   > SetScalarParameterValue(self=@n13, ParameterName=FadeRate, Value=Lerp(A=1, B=3, Alpha=Get CachedFuel()))
   Set IsLowFuel(execute=@n5.else, IsLowFuel:bool=@n4) -> (Output_Get:bool)
   > n14 = UpdateLowFuelState()
   Event Update On Out Of Fuel() {is=title K2Node_CustomEvent}
   > UpdateEmptyState()
   n15 = EventOnVehicleStateChanged!() -> (OutputDelegate, then, PlayerPawn:Fort Player Pawn Object Reference, NewVehicle:Actor Object Reference, OldVehicle:Actor Object Reference) {is=title K2Node_CustomEvent}
   > n16 = Cast To Hoverboard_V2!(execute, Object=@n15.NewVehicle) -> (then, CastFailed, AsHoverboard V2:Hoverboard V2 Object Reference) {is=title K2Node_DynamicCast}
   > SetupVehicle()
   Cast To Hoverboard_V2(execute=@n16.CastFailed, Object=@n15.OldVehicle)
   > ResetVehicle()
   > SetVisibility(InVisibility=Collapsed)
   Event Update On Boost Started() {is=title K2Node_CustomEvent}
   > SetBoostState(bIsBoosting=true)
   n17 = Event PreConstruct()
   > Branch(Condition=@n17.IsDesignTime)
   Event Construct()
   > SetupTouchVisuals()
   n18 = GetOwningPlayer()
   > n19 = Branch(Condition=AND Boolean(A=IsUsingTouch(PlayerController=@n18), B=IsMobileGame()))
   > n20 = Assign On Player Vehicle State Changed!(
    execute:Exec
    self:Fort HUDContext Object Reference=Get FortHUDContext!() -> (ReturnValue:Fort HUDContext Object Reference) {is=title K2Node_GetBlueprintContextEx}
    Delegate:"Delegate (by ref)"=@n15
   ) -> (then:Exec) {is=title K2Node_AssignDelegate}
   SetupVehicle(execute=[@n20.then, @n19.else])
   Event OnWidgetVisibilityChanged!() -> (OutputDelegate, then, bNewVisibility:bool) {is=title K2Node_Event}
   > Set CachedFuel(CachedFuel=-1)
   n21 = Event SetVehicleHasInfiniteFuel!() -> (OutputDelegate, then, bInfiniteFuel:bool) {is=title K2Node_Event}
   > SetActiveWidget(
    self:Widget Switcher Object Reference=Get Switcher_EnergyIcon() -> (Switcher_EnergyIcon:Widget Switcher Object Reference)
    Widget=Select(
     Option 0:Widget Object Reference=Get Overlay_Fuel() -> (Overlay_Fuel:Overlay Object Reference)
     Option 1:Widget Object Reference=Get Overlay_UnlimitedFuel() -> (Overlay_UnlimitedFuel:Overlay Object Reference)
     Index=@n21.bInfiniteFuel
    ) -> (ReturnValue:Widget Object Reference)
   )
  ]}
 ResetVehicle {nodes [
   ResetVehicle() {is=title K2Node_FunctionEntry}
   > n1 = Sequence()
   n2 = Get REF_HoverboardVehicle()
   > Unbind Event from On Boost Started Delegate(self:Fort Athena Vehicle Object Reference=@n2, Delegate=Create Event()) {is=title K2Node_RemoveDelegate}
   > Unbind Event from On Boost Finished Delegate(self:Fort Athena Vehicle Object Reference=@n2, Delegate=Create Event()) {is=title K2Node_RemoveDelegate}
   > Unbind Event from On Low Fuel Delegate(self:Fort Athena Vehicle Object Reference=@n2, Delegate=Create Event()) {is=title K2Node_RemoveDelegate}
   > Unbind Event from On Out Of Fuel Delegate(self:Fort Athena Vehicle Object Reference=@n2, Delegate=Create Event()) {is=title K2Node_RemoveDelegate}
   > Unbind Event from On Fuel Changed Delegate(self:Fort Athena Vehicle Object Reference=@n2, Delegate=Create Event()) {is=title K2Node_RemoveDelegate}
   > Unbind Event from On Disabled Changed Delegate(self:Fort Athena Vehicle Object Reference=@n2, Delegate=Create Event()) {is=title K2Node_RemoveDelegate}
   Set REF_HoverboardVehicle(execute=@n1.then_1, REF_HoverboardVehicle:Hoverboard V2 Object Reference) -> (Output_Get:Hoverboard V2 Object Reference)
  ]}
 SetBoostFXCollapsed {nodes [
   SetBoostFXCollapsed() {is=title K2Node_FunctionEntry}
   > SetVisibility(self=[Get Border_BoostGlow(), Get Overlay_BoostSparks()], InVisibility=Collapsed)
  ]}
 SetBoostState {nodes [
   n1 = SetBoostState!() -> (then, bIsBoosting:bool=~) {is=title K2Node_FunctionEntry}
   > n2 = Branch(Condition=@n1)
   > PlayAnimation(InAnimation=Get Anim_BoostOn() -> (Anim_BoostOn:Widget Animation Object Reference))
   > SetVisibility(self=[Get Border_BoostGlow(), Get Overlay_BoostSparks()], InVisibility=SelfHitTestInvisible)
   n3 = Get Anim_BoostOff() -> (Anim_BoostOff:Widget Animation Object Reference)
   PlayAnimation(execute=@n2.else, InAnimation=@n3)
   > BindToAnimationFinished(Animation=@n3, Delegate=Create Event())
  ]}
 SetupEventBinds {nodes [
   SetupEventBinds() {is=title K2Node_FunctionEntry}
   n1 = Get REF_HoverboardVehicle()
   > Bind Event to On Boost Started Delegate(self:Fort Athena Vehicle Object Reference=@n1, Delegate=Create Event()) {is=title K2Node_AddDelegate}
   > Bind Event to On Boost Finished Delegate(self:Fort Athena Vehicle Object Reference=@n1, Delegate=Create Event()) {is=title K2Node_AddDelegate}
   > Bind Event to On Low Fuel Delegate(self:Fort Athena Vehicle Object Reference=@n1, Delegate=Create Event()) {is=title K2Node_AddDelegate}
   > Bind Event to On Out Of Fuel Delegate(self:Fort Athena Vehicle Object Reference=@n1, Delegate=Create Event()) {is=title K2Node_AddDelegate}
   > Bind Event to On Fuel Changed Delegate(self:Fort Athena Vehicle Object Reference=@n1, Delegate=Create Event()) {is=title K2Node_AddDelegate}
   > Bind Event to On Disabled Changed Delegate(self:Fort Athena Vehicle Object Reference=@n1, Delegate=Create Event()) {is=title K2Node_AddDelegate}
  ]}
 SetupTouchVisuals {nodes [
   SetupTouchVisuals() {is=title K2Node_FunctionEntry}
   n1 = AND Boolean(A=IsUsingTouch(PlayerController=GetOwningPlayer()), B=IsMobileGame())
   > SetHeightOverride(
    self=Get SizeBox_0() -> (SizeBox_0:Size Box Object Reference)
    InHeightOverride=Select(Option 0:float=24, Option 1:float=16, Index=@n1) -> (ReturnValue:float)
   )
   > SetWidthOverride(self=Get SizeBox_0(), InWidthOverride=Select(Option 0=330, Option 1=297, Index=@n1))
   > SetPadding!(
    execute
    self:Overlay Slot Object Reference=SlotAsOverlaySlot(Widget=Get HorizontalBox_Main() -> (HorizontalBox_Main:Horizontal Box Object Reference))
    InPadding_Left:float=0
    InPadding_Top:float=0
    InPadding_Right:float=0
    InPadding_Bottom:float=Select(Option 0=0, Option 1=8, Index=@n1)
   ) -> (then)
   > SetVisibility(
    self=Get Image_TriangleArrow() -> (Image_TriangleArrow:Image Object Reference)
    InVisibility=Select(Option 0:ESlateVisibility Enum=SelfHitTestInvisible, Option 1:ESlateVisibility Enum=Collapsed, Index=@n1) -> (ReturnValue:ESlateVisibility Enum)
   )
   > SetRenderTranslation!(
    execute
    self=Get Overlay_WarningMessaging()
    Translation_X:double=0
    Translation_Y:double=Select(Option 0:double=-56, Option 1:double=0, Index=@n1) -> (ReturnValue:double)
   ) -> (then)
   > SetRenderTranslation!(
    execute
    self=[
     Get Overlay_TextContainerLow() -> (Overlay_TextContainerLow:Overlay Object Reference)
     Get Overlay_TextContainerEject() -> (Overlay_TextContainerEject:Overlay Object Reference)
    ]
    Translation:Vector 2D Structure=Select(Option 0:Vector 2D Structure=(X=8,Y=-8), Option 1:Vector 2D Structure=(X=8,Y=0), Index=@n1) -> (ReturnValue:Vector 2D Structure)
   ) -> (then)
   > SetHorizontalAlignment(
    self=SlotAsOverlaySlot(Widget=Get Overlay_WarningMessaging())
    InHorizontalAlignment=Select(Option 0:EHorizontalAlignment Enum=HAlign_Left, Option 1:EHorizontalAlignment Enum=HAlign_Right, Index=@n1) -> (ReturnValue:EHorizontalAlignment Enum)
   )
   > SetHorizontalAlignment(
    self=[SlotAsOverlaySlot(Widget=Get Overlay_TextContainerEject()), SlotAsOverlaySlot(Widget=Get Overlay_TextContainerLow())]
    InHorizontalAlignment=Select(Option 0=HAlign_Left, Option 1=HAlign_Right, Index=@n1)
   )
   > SetVisibility(
    self=Get Image_AdditonalOuterRing() -> (Image_AdditonalOuterRing:Image Object Reference)
    InVisibility=Select(Option 0:ESlateVisibility Enum=SelfHitTestInvisible, Option 1:ESlateVisibility Enum=Collapsed, Index=@n1) -> (ReturnValue:ESlateVisibility Enum)
   )
   > SetVerticalAlignment(
    self=SlotAsOverlaySlot(Widget=Get Overlay_TextContainerLow())
    InVerticalAlignment=Select(Option 0:EVerticalAlignment Enum=VAlign_Top, Option 1:EVerticalAlignment Enum=VAlign_Center, Index=@n1) -> (ReturnValue:EVerticalAlignment Enum)
   )
   n2 = Select(Option 0:float=0, Option 1:float=4, Index=@n1) -> (ReturnValue:float)
   > SetPadding(
    self=SlotAsOverlaySlot(Widget=Get Image_3() -> (Image_3:Image Object Reference))
    InPadding_Left=@n2
    InPadding_Top=@n2
    InPadding_Right=@n2
    InPadding_Bottom=@n2
   )
   n3 = Select(Option 0=0, Option 1=4, Index=@n1)
   > SetPadding(
    self=SlotAsOverlaySlot(Widget=Get Image() -> (Image:Image Object Reference))
    InPadding_Left=@n3
    InPadding_Top=@n3
    InPadding_Right=@n3
    InPadding_Bottom=@n3
   )
   > SetVisibility(
    self=Get Image_AdditonalOuterRing_1() -> (Image_AdditonalOuterRing_1:Image Object Reference)
    InVisibility=Select(Option 0:ESlateVisibility Enum=SelfHitTestInvisible, Option 1:ESlateVisibility Enum=Collapsed, Index=@n1) -> (ReturnValue:ESlateVisibility Enum)
   )
   > SetVerticalAlignment(
    self=SlotAsOverlaySlot(Widget=Get Overlay_TextContainerEject())
    InVerticalAlignment=Select(Option 0:EVerticalAlignment Enum=VAlign_Top, Option 1:EVerticalAlignment Enum=VAlign_Center, Index=@n1) -> (ReturnValue:EVerticalAlignment Enum)
   )
  ]}
 SetupVehicle {nodes [
   SetupVehicle() {is=title K2Node_FunctionEntry}
   > n1 = Sequence()
   > Is Valid(InputObject=Get REF_HoverboardVehicle())
   > ResetVehicle() {is=title K2Node_CallFunction}
   n2 = Cast To FortPlayerPawn(execute=@n1.then_1, Object=GetOwningPlayerPawn!(self:User Widget Object Reference) -> (ReturnValue:Pawn Object Reference))
   > n3 = Cast To Hoverboard_V2(Object=GetVehicleActor!(self:Fort Player Pawn Object Reference=@n2) -> (ReturnValue:Actor Object Reference))
   > Set REF_HoverboardVehicle(REF_HoverboardVehicle=@n3)
   > SetupEventBinds() {is=title K2Node_CallFunction}
   > UpdateLowFuelState()
   > UpdateEmptyState()
   > SetBoostFXCollapsed() {is=title K2Node_CallFunction}
   > n4 = GetDynamicMaterial(self=Get Image_FillBar())
   > SetScalarParameterValue(self=@n4, ParameterName=Progress, Value=GetVehicleFuelPercent(self=Get REF_HoverboardVehicle()))
   > DisabledChange!(execute, self, bIsDisabled:bool=IsDisabled!(self:Fort Vehicle Interface Interface=Get REF_HoverboardVehicle()) -> (ReturnValue:bool)) -> (then) {is=title K2Node_CallFunction}
   > SetVisibility(InVisibility=HitTestInvisible)
  ]}
 UpdateEmptyState {nodes [
   UpdateEmptyState() {is=title K2Node_FunctionEntry}
   > Branch(Condition=Get IsEmpty() -> (IsEmpty:bool))
   n1 = AND Boolean(A=IsUsingTouch(PlayerController=GetOwningPlayer()), B=IsMobileGame())
   > PlayAnimation(
    InAnimation=Select(
     Option 0:Widget Animation Object Reference=Get Anim_SelfDestructMessaging() -> (Anim_SelfDestructMessaging:Widget Animation Object Reference)
     Option 1:Widget Animation Object Reference=Get Anim_SelfDestructMessaging_Touch() -> (Anim_SelfDestructMessaging_Touch:Widget Animation Object Reference)
     Index=@n1
    ) -> (ReturnValue:Widget Animation Object Reference)
    NumLoopsToPlay=SelectInt(A=0, B=1, bPickA=@n1)
   )
  ]}
 UpdateLowFuelState {nodes [
   UpdateLowFuelState() {is=title K2Node_FunctionEntry}
   > SetVisibility(
    self=[
     Get Overlay_WarningMessaging()
     Get OV_touch_lowwarning() -> (OV_touch_lowwarning:Overlay Object Reference)
     Get Image_NoEnergy_GlowPulse() -> (Image_NoEnergy_GlowPulse:Image Object Reference)
    ]
    InVisibility=Select(Option 0:ESlateVisibility Enum=Collapsed, Option 1:ESlateVisibility Enum=SelfHitTestInvisible, Index=Get IsLowFuel()) -> (ReturnValue:ESlateVisibility Enum)
   )
   n1 = AND Boolean(A=IsUsingTouch(PlayerController=GetOwningPlayer()), B=IsMobileGame())
   > PlayAnimation(
    InAnimation=Select(
     Option 0:Widget Animation Object Reference=Get Anim_LowFuel_Messaging() -> (Anim_LowFuel_Messaging:Widget Animation Object Reference)
     Option 1:Widget Animation Object Reference=Get Anim_LowFuel_Messaging_Touch() -> (Anim_LowFuel_Messaging_Touch:Widget Animation Object Reference)
     Index=@n1
    ) -> (ReturnValue:Widget Animation Object Reference)
    NumLoopsToPlay=SelectInt(A=0, B=1, bPickA=@n1)
   )
  ]}
]
variables [
 BoostGlowBorderScaleEnd {type=number}
 BoostGlowBorderScaleStart {type=number, default=1}
 BoostSparksEndTranslation {type=number, default=-296}
 BoostSparksStartTranslation {type=number}
 CachedFuel {type=number, default=-1}
 IsEmpty {type=boolean}
 IsLowFuel {type=boolean}
 REF_HoverboardVehicle {type=Hoverboard_V2_C, default=null}
]
widgets {
 info {
  inheritedWidgetCount=10
  parentClass=/Script/FortniteUI.FortAthenaVehicleDashboardWidget
  rootWidgetClass=&Overlay
  widgetCount=59
 }
 widgets [
  Overlay_Root {widgetClassPath=&Overlay, widget=Overlay_Root}
  CommonHardwareVisibilityBorder_0 {
   widgetClassPath=&CommonHardwareVisibilityBorder
   parent=Overlay_Root
   widget=CommonHardwareVisibilityBorder_0
   properties {visibilityQuery {
     autoDescription=" NONE( Input.Touch )"
     tokenStreamVersion=0
     userDescription=Not touch
     queryTokenStream [0,1,3,1,0]
     tagDictionary [{tagName=Input.Touch}]
    }}
   slot {class=OverlaySlot}
  }
  WBP_Vehicle_Warning {
   widgetClassPath=/Game/Athena/HUD/Vehicles/Dashboard/WBP_Vehicle_Warning
   bIsVariable=true
   parent=CommonHardwareVisibilityBorder_0
   widget=WBP_Vehicle_Warning
   properties {visibility=Collapsed, renderTransform {angle=0, scale=<0.75 0.75>, shear=<0 0>, translation=<-24 0>}}
   slot {class=BorderSlot, properties {padding=<ltrb 0 0 0 0>}}
  }
  HorizontalBox_Main {
   widgetClassPath=&HorizontalBox
   bIsVariable=true
   parent=Overlay_Root
   widget=HorizontalBox_Main
   slot {class=OverlaySlot, properties {verticalAlignment=VAlign_Bottom}}
  }
  Switcher_EnergyIcon {
   widgetClassPath=&WidgetSwitcher
   bIsVariable=true
   parent=HorizontalBox_Main
   widget=Switcher_EnergyIcon
   slot {class=HorizontalBoxSlot, properties {horizontalAlignment=HAlign_Center, verticalAlignment=VAlign_Center, padding=<ltrb -14 -12 0 -12>}}
  }
  Overlay_Fuel {widgetClassPath=&Overlay, bIsVariable=true, parent=Switcher_EnergyIcon, widget=Overlay_Fuel, slot {class=WidgetSwitcherSlot}}
  Image_NoEnergy_GlowPulse {
   widgetClassPath=&Image
   bIsVariable=true
   parent=Overlay_Fuel
   widget=Image_NoEnergy_GlowPulse
   properties {
    renderOpacity=0.5
    visibility=SelfHitTestInvisible
    brush {
     bIsDynamicallyLoaded=false
     drawAs=Image
     imageType=NoImage
     mirroring=NoMirror
     resourceName=None
     resourceObject=/Game/Athena/HUD/LTM/Wax/Materials/M_UI_TextGlow_Add_Gradient_Orange_Nevada
     tiling=NoTile
     imageSize=<24 24>
     margin=<ltrb 0 0 0 0>
     outlineSettings {
      bUseBrushTransparency=false
      roundingType=HalfHeightRadius
      width=0
      color {colorUseRule=UseColor_Specified, specifiedColor=<rgba 0 0 0 0>}
      cornerRadii=<0 0 0 0>
     }
     tintColor {colorUseRule=UseColor_Specified, specifiedColor=<rgba 1 1 1 1>}
     uVRegion {bIsValid=false, max=<0 0>, min=<0 0>}
    }
    renderTransform {angle=0, scale=<1.325 1.75>, shear=<0 0>, translation=<0 0>}
   }
   slot {class=OverlaySlot, properties {horizontalAlignment=HAlign_Fill, verticalAlignment=VAlign_Fill}}
  }
  Image_Energy {
   widgetClassPath=&Image
   parent=Overlay_Fuel
   widget=Image_Energy
   properties {
    brush {
     bIsDynamicallyLoaded=false
     drawAs=Image
     imageType=NoImage
     mirroring=NoMirror
     resourceName=None
     resourceObject=/Game/UI/Foundation/Textures/Icons/Stats/T-Icon-HealthBattery-64
     tiling=NoTile
     imageSize=<28 28>
     margin=<ltrb 0 0 0 0>
     outlineSettings {
      bUseBrushTransparency=false
      roundingType=HalfHeightRadius
      width=0
      color {colorUseRule=UseColor_Specified, specifiedColor=<rgba 0 0 0 0>}
      cornerRadii=<0 0 0 0>
     }
     tintColor {colorUseRule=UseColor_Specified, specifiedColor=<rgba 1 1 1 1>}
     uVRegion {bIsValid=false, max=<0 0>, min=<0 0>}
    }
    renderTransform {angle=0, scale=<1.1 1.1>, shear=<0 0>, translation=<0 0>}
   }
   slot {class=OverlaySlot, properties {horizontalAlignment=HAlign_Center, verticalAlignment=VAlign_Center}}
  }
  Overlay_UnlimitedFuel {
   widgetClassPath=&Overlay
   bIsVariable=true
   parent=Switcher_EnergyIcon
   widget=Overlay_UnlimitedFuel
   slot {class=WidgetSwitcherSlot}
  }
  SizeBox_2 {
   widgetClassPath=&SizeBox
   parent=Overlay_UnlimitedFuel
   widget=SizeBox_2
   properties {bOverride_HeightOverride=true, bOverride_WidthOverride=true, heightOverride=28, widthOverride=28}
   slot {class=OverlaySlot, properties {horizontalAlignment=HAlign_Center, verticalAlignment=VAlign_Center}}
  }
  CRichText_Infinite {
   widgetClassPath=&CommonRichTextBlock
   parent=SizeBox_2
   widget=CRichText_Infinite
   properties {
    bOverrideDefaultStyle=true
    justification=Right
    lineHeightPercentage=0.85
    text="∞"
    textTransformPolicy=ToUpper
    defaultTextStyleOverride {
     overflowPolicy=Clip
     transformPolicy=None
     colorAndOpacity {colorUseRule=UseColor_Specified, specifiedColor=<rgba 1 1 1 1>}
     font {
      bForceMonospaced=false
      bMaterialIsStencil=false
      fontMaterial=null
      fontObject=&BurbankSmall
      letterSpacing=0
      monospacedWidth=1
      size=16.5
      skewAmount=0
      typefaceFontName=Bold
      outlineSettings {
       bApplyOutlineToDropShadows=false
       bMiteredCorners=false
       bSeparateFillAlpha=false
       outlineMaterial=null
       outlineSize=0
       outlineColor=<rgba 0 0 0 1>
      }
     }
     highlightColor {colorUseRule=UseColor_Specified, specifiedColor=<rgba 0 0 0 1>}
     highlightShape {
      bIsDynamicallyLoaded=false
      drawAs=Image
      imageType=NoImage
      mirroring=NoMirror
      resourceName=None
      resourceObject=null
      tiling=NoTile
      imageSize=<32 32>
      margin=<ltrb 0 0 0 0>
      outlineSettings {
       bUseBrushTransparency=false
       roundingType=HalfHeightRadius
       width=0
       color {colorUseRule=UseColor_Specified, specifiedColor=<rgba 0 0 0 0>}
       cornerRadii=<0 0 0 0>
      }
      tintColor {colorUseRule=UseColor_Specified, specifiedColor=<rgba 1 1 1 1>}
      uVRegion {bIsValid=false, max=<0 0>, min=<0 0>}
     }
     selectedBackgroundColor {colorUseRule=UseColor_Foreground, specifiedColor=<rgba 1 0 1 1>}
     shadowColorAndOpacity=<rgba 0 0 0 1>
     shadowOffset=<0 0>
     strikeBrush {
      bIsDynamicallyLoaded=false
      drawAs=Image
      imageType=NoImage
      mirroring=NoMirror
      resourceName=None
      resourceObject=null
      tiling=NoTile
      imageSize=<32 32>
      margin=<ltrb 0 0 0 0>
      outlineSettings {
       bUseBrushTransparency=false
       roundingType=HalfHeightRadius
       width=0
       color {colorUseRule=UseColor_Specified, specifiedColor=<rgba 0 0 0 0>}
       cornerRadii=<0 0 0 0>
      }
      tintColor {colorUseRule=UseColor_Specified, specifiedColor=<rgba 1 1 1 1>}
      uVRegion {bIsValid=false, max=<0 0>, min=<0 0>}
     }
     underlineBrush {
      bIsDynamicallyLoaded=false
      drawAs=Image
      imageType=NoImage
      mirroring=NoMirror
      resourceName=None
      resourceObject=null
      tiling=NoTile
      imageSize=<32 32>
      margin=<ltrb 0 0 0 0>
      outlineSettings {
       bUseBrushTransparency=false
       roundingType=HalfHeightRadius
       width=0
       color {colorUseRule=UseColor_Specified, specifiedColor=<rgba 0 0 0 0>}
       cornerRadii=<0 0 0 0>
      }
      tintColor {colorUseRule=UseColor_Specified, specifiedColor=<rgba 1 1 1 1>}
      uVRegion {bIsValid=false, max=<0 0>, min=<0 0>}
     }
    }
    renderTransform {angle=0, scale=<1 1>, shear=<0 0>, translation=<0 -3>}
   }
   slot {class=SizeBoxSlot, properties {horizontalAlignment=HAlign_Center, verticalAlignment=VAlign_Center}}
  }
  Overlay_FuelBar {
   widgetClassPath=&Overlay
   parent=HorizontalBox_Main
   widget=Overlay_FuelBar
   slot {class=HorizontalBoxSlot, properties {verticalAlignment=VAlign_Center}}
  }
  SizeBox_0 {
   widgetClassPath=&SizeBox
   bIsVariable=true
   parent=Overlay_FuelBar
   widget=SizeBox_0
   properties {bOverride_HeightOverride=true, bOverride_WidthOverride=true, heightOverride=24, widthOverride=330}
   slot {class=OverlaySlot, properties {horizontalAlignment=HAlign_Fill, verticalAlignment=VAlign_Center}}
  }
  Overlay_0 {widgetClassPath=&Overlay, parent=SizeBox_0, widget=Overlay_0, slot {class=SizeBoxSlot}}
  Image_FillBar {
   widgetClassPath=&Image
   bIsVariable=true
   parent=Overlay_0
   widget=Image_FillBar
   properties {brush {
     bIsDynamicallyLoaded=false
     drawAs=Image
     imageType=NoImage
     mirroring=NoMirror
     resourceName=None
     resourceObject=/HoverboardV2/HUD/Materials/PROTO_M_UI_Hoverboard_V2_EnergyBar
     tiling=NoTile
     imageSize=<32 32>
     margin=<ltrb 0 0 0 0>
     outlineSettings {
      bUseBrushTransparency=false
      roundingType=HalfHeightRadius
      width=0
      color {colorUseRule=UseColor_Specified, specifiedColor=<rgba 0 0 0 0>}
      cornerRadii=<0 0 0 0>
     }
     tintColor {colorUseRule=UseColor_Specified, specifiedColor=<rgba 1 1 1 1>}
     uVRegion {bIsValid=false, max=<0 0>, min=<0 0>}
    }}
   slot {class=OverlaySlot, properties {horizontalAlignment=HAlign_Fill, verticalAlignment=VAlign_Fill}}
  }
  OV_touch_lowwarning {
   widgetClassPath=&Overlay
   bIsVariable=true
   parent=Overlay_0
   widget=OV_touch_lowwarning
   properties {renderOpacity=0, visibility=Collapsed}
   slot {class=OverlaySlot, properties {horizontalAlignment=HAlign_Right, verticalAlignment=VAlign_Center}}
  }
  bg_low {
   widgetClassPath=&Image
   parent=OV_touch_lowwarning
   widget=bg_low
   properties {visibility=SelfHitTestInvisible, colorAndOpacity=<rgba 1 0.3185 0 1>}
   slot {class=OverlaySlot, properties {horizontalAlignment=HAlign_Fill, verticalAlignment=VAlign_Fill}}
  }
  bg_low_skew {
   widgetClassPath=&Image
   parent=OV_touch_lowwarning
   widget=bg_low_skew
   properties {
    visibility=SelfHitTestInvisible
    brush {
     bIsDynamicallyLoaded=false
     drawAs=Image
     imageType=NoImage
     mirroring=NoMirror
     resourceName=None
     resourceObject=null
     tiling=NoTile
     imageSize=<16 16>
     margin=<ltrb 0 0 0 0>
     outlineSettings {
      bUseBrushTransparency=false
      roundingType=HalfHeightRadius
      width=0
      color {colorUseRule=UseColor_Specified, specifiedColor=<rgba 0 0 0 0>}
      cornerRadii=<0 0 0 0>
     }
     tintColor {colorUseRule=UseColor_Specified, specifiedColor=<rgba 1 1 1 1>}
     uVRegion {bIsValid=false, max=<0 0>, min=<0 0>}
    }
    colorAndOpacity=<rgba 1 0.3185 0 1>
    renderTransform {angle=0, scale=<1 1>, shear=<-25 0>, translation=<0 0>}
   }
   slot {class=OverlaySlot, properties {verticalAlignment=VAlign_Fill, padding=<ltrb -5 0 0 0>}}
  }
  SizeBox_4 {
   widgetClassPath=&SizeBox
   parent=OV_touch_lowwarning
   widget=SizeBox_4
   properties {bOverride_MaxDesiredWidth=true, clipping=ClipToBounds, maxDesiredWidth=170}
   slot {class=OverlaySlot, properties {horizontalAlignment=HAlign_Right, verticalAlignment=VAlign_Center, padding=<ltrb 5 0 5 0>}}
  }
  lowenergy_touch {
   widgetClassPath=&CommonTextBlock
   parent=SizeBox_4
   widget=lowenergy_touch
   properties {
    justification=Center
    scrollStyle=&TextScrollStyle-Base_C
    text=Low Energy
    textTransformPolicy=ToUpper
    colorAndOpacity {colorUseRule=UseColor_Specified, specifiedColor=<rgba 0 0 0 1>}
    font {
     bForceMonospaced=false
     bMaterialIsStencil=false
     fontMaterial=null
     fontObject=&BurbankSmall
     letterSpacing=100
     monospacedWidth=1
     size=12
     skewAmount=0
     typefaceFontName=Black
     outlineSettings {
      bApplyOutlineToDropShadows=false
      bMiteredCorners=false
      bSeparateFillAlpha=true
      outlineMaterial=null
      outlineSize=2
      outlineColor=<rgba 1 1 1 0.125>
     }
    }
    renderTransform {angle=0, scale=<1 1>, shear=<0 0>, translation=<0 -3>}
    shadowOffset=<0 0>
   }
   slot {class=SizeBoxSlot}
  }
  OV_touch_emptywarning {
   widgetClassPath=&Overlay
   bIsVariable=true
   parent=Overlay_0
   widget=OV_touch_emptywarning
   properties {renderOpacity=0}
   slot {class=OverlaySlot, properties {horizontalAlignment=HAlign_Right, verticalAlignment=VAlign_Center}}
  }
  bg_empty {
   widgetClassPath=&Image
   parent=OV_touch_emptywarning
   widget=bg_empty
   properties {visibility=SelfHitTestInvisible, colorAndOpacity=<rgba 1 0.0065 0 1>}
   slot {class=OverlaySlot, properties {horizontalAlignment=HAlign_Fill, verticalAlignment=VAlign_Fill}}
  }
  bg_empty_skew {
   widgetClassPath=&Image
   parent=OV_touch_emptywarning
   widget=bg_empty_skew
   properties {
    visibility=SelfHitTestInvisible
    brush {
     bIsDynamicallyLoaded=false
     drawAs=Image
     imageType=NoImage
     mirroring=NoMirror
     resourceName=None
     resourceObject=null
     tiling=NoTile
     imageSize=<16 16>
     margin=<ltrb 0 0 0 0>
     outlineSettings {
      bUseBrushTransparency=false
      roundingType=HalfHeightRadius
      width=0
      color {colorUseRule=UseColor_Specified, specifiedColor=<rgba 0 0 0 0>}
      cornerRadii=<0 0 0 0>
     }
     tintColor {colorUseRule=UseColor_Specified, specifiedColor=<rgba 1 1 1 1>}
     uVRegion {bIsValid=false, max=<0 0>, min=<0 0>}
    }
    colorAndOpacity=<rgba 1 0.0065 0 1>
    renderTransform {angle=0, scale=<1 1>, shear=<-25 0>, translation=<0 0>}
   }
   slot {class=OverlaySlot, properties {verticalAlignment=VAlign_Fill, padding=<ltrb -5 0 0 0>}}
  }
  SizeBox_3 {
   widgetClassPath=&SizeBox
   parent=OV_touch_emptywarning
   widget=SizeBox_3
   properties {bOverride_MaxDesiredWidth=true, clipping=ClipToBounds, maxDesiredWidth=170, widthOverride=170}
   slot {class=OverlaySlot, properties {horizontalAlignment=HAlign_Right, verticalAlignment=VAlign_Center, padding=<ltrb 5 0 5 0>}}
  }
  empty_touch {
   widgetClassPath=&CommonTextBlock
   parent=SizeBox_3
   widget=empty_touch
   properties {
    justification=Center
    scrollStyle=&TextScrollStyle-Base_C
    text=ENERGY DEPLETED
    textTransformPolicy=ToUpper
    colorAndOpacity {colorUseRule=UseColor_Specified, specifiedColor=<rgba 0 0 0 1>}
    font {
     bForceMonospaced=false
     bMaterialIsStencil=false
     fontMaterial=null
     fontObject=&BurbankSmall
     letterSpacing=100
     monospacedWidth=1
     size=12
     skewAmount=0
     typefaceFontName=Black
     outlineSettings {
      bApplyOutlineToDropShadows=false
      bMiteredCorners=false
      bSeparateFillAlpha=true
      outlineMaterial=null
      outlineSize=2
      outlineColor=<rgba 1 1 1 0.125>
     }
    }
    renderTransform {angle=0, scale=<1 1>, shear=<0 0>, translation=<0 -3>}
    shadowOffset=<0 0>
   }
   slot {class=SizeBoxSlot}
  }
  Overlay_WarningMessaging {
   widgetClassPath=&Overlay
   bIsVariable=true
   parent=Overlay_FuelBar
   widget=Overlay_WarningMessaging
   properties {renderOpacity=0, visibility=Collapsed, renderTransform {angle=0, scale=<1 1>, shear=<-16 0>, translation=<0 -56>}}
   slot {class=OverlaySlot, properties {verticalAlignment=VAlign_Center}}
  }
  Overlay_4 {
   widgetClassPath=&Overlay
   parent=Overlay_WarningMessaging
   widget=Overlay_4
   slot {class=OverlaySlot, properties {horizontalAlignment=HAlign_Fill, verticalAlignment=VAlign_Fill, padding=<ltrb -8 -8 -8 -8>}}
  }
  Overlay_TextContainerLow {
   widgetClassPath=&Overlay
   bIsVariable=true
   parent=Overlay_4
   widget=Overlay_TextContainerLow
   properties {renderTransform {angle=0, scale=<1 1>, shear=<0 0>, translation=<8 -8>}}
   slot {class=OverlaySlot}
  }
  Image_AdditonalOuterRing {
   widgetClassPath=&Image
   bIsVariable=true
   parent=Overlay_TextContainerLow
   widget=Image_AdditonalOuterRing
   properties {
    renderOpacity=0.25
    brush {
     bIsDynamicallyLoaded=false
     drawAs=Box
     imageType=NoImage
     mirroring=NoMirror
     resourceName=None
     resourceObject=/Game/Athena/HUD/Vehicles/Materials/M_UI_OstrichWarningText_Outline
     tiling=NoTile
     imageSize=<16 16>
     margin=<ltrb 0.5 0.5 0.5 0.5>
     outlineSettings {
      bUseBrushTransparency=false
      roundingType=HalfHeightRadius
      width=0
      color {colorUseRule=UseColor_Specified, specifiedColor=<rgba 0 0 0 0>}
      cornerRadii=<0 0 0 0>
     }
     tintColor {colorUseRule=UseColor_Specified, specifiedColor=<rgba 1 1 1 1>}
     uVRegion {bIsValid=false, max=<0 0>, min=<0 0>}
    }
   }
   slot {class=OverlaySlot, properties {horizontalAlignment=HAlign_Fill, verticalAlignment=VAlign_Fill, padding=<ltrb -8 -8 -8 -8>}}
  }
  Image_3 {
   widgetClassPath=&Image
   bIsVariable=true
   parent=Overlay_TextContainerLow
   widget=Image_3
   properties {
    brush {
     bIsDynamicallyLoaded=false
     drawAs=Box
     imageType=NoImage
     mirroring=NoMirror
     resourceName=None
     resourceObject=/Game/Athena/HUD/Vehicles/Materials/M_UI_OstrichWarningText_Backing
     tiling=NoTile
     imageSize=<8 8>
     margin=<ltrb 0.5 0.5 0.5 0.5>
     outlineSettings {
      bUseBrushTransparency=false
      roundingType=HalfHeightRadius
      width=0
      color {colorUseRule=UseColor_Specified, specifiedColor=<rgba 0 0 0 0>}
      cornerRadii=<0 0 0 0>
     }
     tintColor {colorUseRule=UseColor_Specified, specifiedColor=<rgba 1 1 1 1>}
     uVRegion {bIsValid=false, max=<0 0>, min=<0 0>}
    }
    renderTransformPivot=<1 0.5>
   }
   slot {class=OverlaySlot, properties {horizontalAlignment=HAlign_Fill, verticalAlignment=VAlign_Fill}}
  }
  Image_4 {
   widgetClassPath=&Image
   parent=Overlay_TextContainerLow
   widget=Image_4
   properties {
    renderOpacity=0.5
    brush {
     bIsDynamicallyLoaded=false
     drawAs=Box
     imageType=NoImage
     mirroring=NoMirror
     resourceName=None
     resourceObject=&M_UI_OstrichWarningText_Outline
     tiling=NoTile
     imageSize=<8 8>
     margin=<ltrb 0.5 0.5 0.5 0.5>
     outlineSettings {
      bUseBrushTransparency=false
      roundingType=HalfHeightRadius
      width=0
      color {colorUseRule=UseColor_Specified, specifiedColor=<rgba 0 0 0 0>}
      cornerRadii=<0 0 0 0>
     }
     tintColor {colorUseRule=UseColor_Specified, specifiedColor=<rgba 1 1 1 1>}
     uVRegion {bIsValid=false, max=<0 0>, min=<0 0>}
    }
   }
   slot {class=OverlaySlot, properties {horizontalAlignment=HAlign_Fill, verticalAlignment=VAlign_Fill}}
  }
  Text_NoEnergy {
   widgetClassPath=&CommonTextBlock
   bIsVariable=true
   parent=Overlay_TextContainerLow
   widget=Text_NoEnergy
   properties {
    justification=Center
    mobileFontSizeMultiplier=0.7
    text="Low Energy!"
    textTransformPolicy=ToUpper
    colorAndOpacity {colorUseRule=UseColor_Specified, specifiedColor=<rgba 0 0 0 0.5>}
    font {
     bForceMonospaced=false
     bMaterialIsStencil=false
     fontMaterial=null
     fontObject=&BurbankSmall
     letterSpacing=100
     monospacedWidth=1
     size=13
     skewAmount=0
     typefaceFontName=Black
     outlineSettings {
      bApplyOutlineToDropShadows=false
      bMiteredCorners=false
      bSeparateFillAlpha=true
      outlineMaterial=null
      outlineSize=2
      outlineColor=<rgba 1 1 1 0.125>
     }
    }
    shadowOffset=<0 0>
   }
   slot {class=OverlaySlot, properties {horizontalAlignment=HAlign_Fill, verticalAlignment=VAlign_Center, padding=<ltrb 12 2 12 2>}}
  }
  Image_LargeGlow {
   widgetClassPath=&Image
   parent=Overlay_TextContainerLow
   widget=Image_LargeGlow
   properties {
    renderOpacity=0.25
    visibility=SelfHitTestInvisible
    brush {
     bIsDynamicallyLoaded=false
     drawAs=Box
     imageType=NoImage
     mirroring=NoMirror
     resourceName=None
     resourceObject=&M_UI_TextGlow_Add_Gradient_Orange_Nevada
     tiling=NoTile
     imageSize=<64 64>
     margin=<ltrb 0.5 0.5 0.5 0.5>
     outlineSettings {
      bUseBrushTransparency=false
      roundingType=HalfHeightRadius
      width=0
      color {colorUseRule=UseColor_Specified, specifiedColor=<rgba 0 0 0 0>}
      cornerRadii=<0 0 0 0>
     }
     tintColor {colorUseRule=UseColor_Specified, specifiedColor=<rgba 1 1 1 1>}
     uVRegion {bIsValid=false, max=<0 0>, min=<0 0>}
    }
   }
   slot {class=OverlaySlot, properties {horizontalAlignment=HAlign_Fill, verticalAlignment=VAlign_Fill, padding=<ltrb -24 -24 -24 -24>}}
  }
  Overlay_TextContainerEject {
   widgetClassPath=&Overlay
   bIsVariable=true
   parent=Overlay_4
   widget=Overlay_TextContainerEject
   properties {renderOpacity=0, renderTransform {angle=0, scale=<1 1>, shear=<0 0>, translation=<8 -8>}}
   slot {class=OverlaySlot}
  }
  Image_AdditonalOuterRing_1 {
   widgetClassPath=&Image
   bIsVariable=true
   parent=Overlay_TextContainerEject
   widget=Image_AdditonalOuterRing_1
   properties {
    renderOpacity=0.25
    brush {
     bIsDynamicallyLoaded=false
     drawAs=Box
     imageType=NoImage
     mirroring=NoMirror
     resourceName=None
     resourceObject=&M_UI_OstrichWarningText_Outline
     tiling=NoTile
     imageSize=<16 16>
     margin=<ltrb 0.5 0.5 0.5 0.5>
     outlineSettings {
      bUseBrushTransparency=false
      roundingType=HalfHeightRadius
      width=0
      color {colorUseRule=UseColor_Specified, specifiedColor=<rgba 0 0 0 0>}
      cornerRadii=<0 0 0 0>
     }
     tintColor {colorUseRule=UseColor_Specified, specifiedColor=<rgba 1 1 1 1>}
     uVRegion {bIsValid=false, max=<0 0>, min=<0 0>}
    }
   }
   slot {class=OverlaySlot, properties {horizontalAlignment=HAlign_Fill, verticalAlignment=VAlign_Fill, padding=<ltrb -8 -8 -8 -8>}}
  }
  Image {
   widgetClassPath=&Image
   bIsVariable=true
   parent=Overlay_TextContainerEject
   widget=Image
   properties {
    brush {
     bIsDynamicallyLoaded=false
     drawAs=Box
     imageType=NoImage
     mirroring=NoMirror
     resourceName=None
     resourceObject=/Game/Athena/HUD/Vehicles/Materials/M_UI_OstrichWarningText_Backing_Red
     tiling=NoTile
     imageSize=<8 8>
     margin=<ltrb 0.5 0.5 0.5 0.5>
     outlineSettings {
      bUseBrushTransparency=false
      roundingType=HalfHeightRadius
      width=0
      color {colorUseRule=UseColor_Specified, specifiedColor=<rgba 0 0 0 0>}
      cornerRadii=<0 0 0 0>
     }
     tintColor {colorUseRule=UseColor_Specified, specifiedColor=<rgba 1 1 1 1>}
     uVRegion {bIsValid=false, max=<0 0>, min=<0 0>}
    }
    renderTransformPivot=<1 0.5>
   }
   slot {class=OverlaySlot, properties {horizontalAlignment=HAlign_Fill, verticalAlignment=VAlign_Fill}}
  }
  Image_1 {
   widgetClassPath=&Image
   parent=Overlay_TextContainerEject
   widget=Image_1
   properties {
    renderOpacity=0.5
    brush {
     bIsDynamicallyLoaded=false
     drawAs=Box
     imageType=NoImage
     mirroring=NoMirror
     resourceName=None
     resourceObject=&M_UI_OstrichWarningText_Outline
     tiling=NoTile
     imageSize=<8 8>
     margin=<ltrb 0.5 0.5 0.5 0.5>
     outlineSettings {
      bUseBrushTransparency=false
      roundingType=HalfHeightRadius
      width=0
      color {colorUseRule=UseColor_Specified, specifiedColor=<rgba 0 0 0 0>}
      cornerRadii=<0 0 0 0>
     }
     tintColor {colorUseRule=UseColor_Specified, specifiedColor=<rgba 1 1 1 1>}
     uVRegion {bIsValid=false, max=<0 0>, min=<0 0>}
    }
   }
   slot {class=OverlaySlot, properties {horizontalAlignment=HAlign_Fill, verticalAlignment=VAlign_Fill}}
  }
  Text_Eject {
   widgetClassPath=&CommonTextBlock
   parent=Overlay_TextContainerEject
   widget=Text_Eject
   properties {
    justification=Center
    mobileFontSizeMultiplier=0.7
    text=ENERGY DEPLETED
    textTransformPolicy=ToUpper
    colorAndOpacity {colorUseRule=UseColor_Specified, specifiedColor=<rgba 0 0 0 0.5>}
    font {
     bForceMonospaced=false
     bMaterialIsStencil=false
     fontMaterial=null
     fontObject=&BurbankSmall
     letterSpacing=100
     monospacedWidth=1
     size=13
     skewAmount=0
     typefaceFontName=Black
     outlineSettings {
      bApplyOutlineToDropShadows=false
      bMiteredCorners=false
      bSeparateFillAlpha=true
      outlineMaterial=null
      outlineSize=2
      outlineColor=<rgba 1 1 1 0.125>
     }
    }
    shadowOffset=<0 0>
   }
   slot {class=OverlaySlot, properties {horizontalAlignment=HAlign_Fill, verticalAlignment=VAlign_Center, padding=<ltrb 12 2 12 2>}}
  }
  Image_LargeGlow_1 {
   widgetClassPath=&Image
   parent=Overlay_TextContainerEject
   widget=Image_LargeGlow_1
   properties {
    renderOpacity=0.25
    visibility=SelfHitTestInvisible
    brush {
     bIsDynamicallyLoaded=false
     drawAs=Box
     imageType=NoImage
     mirroring=NoMirror
     resourceName=None
     resourceObject=/Game/Athena/HUD/LTM/Wax/Materials/M_UI_BallerSelftDestruct_Textglow
     tiling=NoTile
     imageSize=<64 64>
     margin=<ltrb 0.5 0.5 0.5 0.5>
     outlineSettings {
      bUseBrushTransparency=false
      roundingType=HalfHeightRadius
      width=0
      color {colorUseRule=UseColor_Specified, specifiedColor=<rgba 0 0 0 0>}
      cornerRadii=<0 0 0 0>
     }
     tintColor {colorUseRule=UseColor_Specified, specifiedColor=<rgba 1 1 1 1>}
     uVRegion {bIsValid=false, max=<0 0>, min=<0 0>}
    }
   }
   slot {class=OverlaySlot, properties {horizontalAlignment=HAlign_Fill, verticalAlignment=VAlign_Fill, padding=<ltrb -24 -24 -24 -24>}}
  }
  Overlay_AlertIcon {
   widgetClassPath=&Overlay
   parent=Overlay_TextContainerEject
   widget=Overlay_AlertIcon
   properties {visibility=Collapsed, renderTransform {angle=0, scale=<1 1>, shear=<8 0>, translation=<-32 -40>}}
   slot {class=OverlaySlot}
  }
  Image_210 {
   widgetClassPath=&Image
   parent=Overlay_AlertIcon
   widget=Image_210
   properties {brush {
     bIsDynamicallyLoaded=false
     drawAs=Image
     imageType=NoImage
     mirroring=NoMirror
     resourceName=None
     resourceObject=&M_UI_OstrichWarningText_Backing
     tiling=NoTile
     imageSize=<32 32>
     margin=<ltrb 0 0 0 0>
     outlineSettings {
      bUseBrushTransparency=false
      roundingType=HalfHeightRadius
      width=0
      color {colorUseRule=UseColor_Specified, specifiedColor=<rgba 0 0 0 0>}
      cornerRadii=<0 0 0 0>
     }
     tintColor {colorUseRule=UseColor_Specified, specifiedColor=<rgba 1 1 1 1>}
     uVRegion {bIsValid=false, max=<0 0>, min=<0 0>}
    }}
   slot {class=OverlaySlot}
  }
  CommonTextBlock_245 {
   widgetClassPath=&CommonTextBlock
   parent=Overlay_AlertIcon
   widget=CommonTextBlock_245
   properties {
    text="!"
    font {
     bForceMonospaced=false
     bMaterialIsStencil=false
     fontMaterial=null
     fontObject=&BurbankSmall
     letterSpacing=0
     monospacedWidth=1
     size=17
     skewAmount=0
     typefaceFontName=Black
     outlineSettings {
      bApplyOutlineToDropShadows=false
      bMiteredCorners=false
      bSeparateFillAlpha=true
      outlineMaterial=null
      outlineSize=2
      outlineColor=<rgba 1 1 1 0.125>
     }
    }
    shadowOffset=<0 0>
   }
   slot {class=OverlaySlot, properties {horizontalAlignment=HAlign_Center, verticalAlignment=VAlign_Center}}
  }
  Image_LargeGlow_2 {
   widgetClassPath=&Image
   parent=Overlay_AlertIcon
   widget=Image_LargeGlow_2
   properties {
    renderOpacity=0.25
    visibility=SelfHitTestInvisible
    brush {
     bIsDynamicallyLoaded=false
     drawAs=Image
     imageType=NoImage
     mirroring=NoMirror
     resourceName=None
     resourceObject=&M_UI_BallerSelftDestruct_Textglow
     tiling=NoTile
     imageSize=<64 64>
     margin=<ltrb 0.5 0.5 0.5 0.5>
     outlineSettings {
      bUseBrushTransparency=false
      roundingType=HalfHeightRadius
      width=0
      color {colorUseRule=UseColor_Specified, specifiedColor=<rgba 0 0 0 0>}
      cornerRadii=<0 0 0 0>
     }
     tintColor {colorUseRule=UseColor_Specified, specifiedColor=<rgba 1 1 1 1>}
     uVRegion {bIsValid=false, max=<0 0>, min=<0 0>}
    }
   }
   slot {class=OverlaySlot, properties {horizontalAlignment=HAlign_Fill, verticalAlignment=VAlign_Fill, padding=<ltrb -24 -24 -24 -24>}}
  }
  Image_TriangleArrow {
   widgetClassPath=&Image
   bIsVariable=true
   parent=Overlay_WarningMessaging
   widget=Image_TriangleArrow
   properties {
    visibility=SelfHitTestInvisible
    brush {
     bIsDynamicallyLoaded=false
     drawAs=Image
     imageType=NoImage
     mirroring=NoMirror
     resourceName=None
     resourceObject=/Game/UI/SocialPanel/Art/Textures/T_UI_Icon_TreeArrow_64
     tiling=NoTile
     imageSize=<16 16>
     margin=<ltrb 0 0 0 0>
     outlineSettings {
      bUseBrushTransparency=false
      roundingType=HalfHeightRadius
      width=0
      color {colorUseRule=UseColor_Specified, specifiedColor=<rgba 0 0 0 0>}
      cornerRadii=<0 0 0 0>
     }
     tintColor {colorUseRule=UseColor_Specified, specifiedColor=<rgba 1 1 1 1>}
     uVRegion {bIsValid=false, max=<0 0>, min=<0 0>}
    }
    colorAndOpacity=<rgba 1 0.4167 0 1>
    renderTransform {angle=108, scale=<1 1>, shear=<0 0>, translation=<-8 24>}
   }
   slot {class=OverlaySlot, properties {verticalAlignment=VAlign_Bottom}}
  }
  Border_BoostGlow {
   widgetClassPath=&Border
   bIsVariable=true
   parent=Overlay_FuelBar
   widget=Border_BoostGlow
   properties {
    horizontalAlignment=HAlign_Left
    visibility=Collapsed
    background {
     bIsDynamicallyLoaded=false
     drawAs=NoDrawType
     imageType=NoImage
     mirroring=NoMirror
     resourceName=None
     resourceObject=null
     tiling=NoTile
     imageSize=<32 32>
     margin=<ltrb 0 0 0 0>
     outlineSettings {
      bUseBrushTransparency=false
      roundingType=HalfHeightRadius
      width=0
      color {colorUseRule=UseColor_Specified, specifiedColor=<rgba 0 0 0 0>}
      cornerRadii=<0 0 0 0>
     }
     tintColor {colorUseRule=UseColor_Specified, specifiedColor=<rgba 1 1 1 1>}
     uVRegion {bIsValid=false, max=<0 0>, min=<0 0>}
    }
    padding=<ltrb 0 0 0 0>
    renderTransformPivot=<0 0.5>
   }
   slot {class=OverlaySlot, properties {verticalAlignment=VAlign_Fill}}
  }
  SizeBox_1 {
   widgetClassPath=&SizeBox
   parent=Border_BoostGlow
   widget=SizeBox_1
   properties {bOverride_WidthOverride=true, widthOverride=330}
   slot {class=BorderSlot, properties {horizontalAlignment=HAlign_Left, padding=<ltrb 0 0 0 0>}}
  }
  Image_BoostGlow {
   widgetClassPath=&Image
   bInherited=true
   bIsVariable=true
   parent=SizeBox_1
   widget=Image_BoostGlow
   properties {
    brush {
     bIsDynamicallyLoaded=false
     drawAs=Box
     imageType=NoImage
     mirroring=NoMirror
     resourceName=None
     resourceObject=/Game/Athena/HUD/LTM/Wax/Materials/M_UI_BallerBoostGlow
     tiling=NoTile
     imageSize=<72 72>
     margin=<ltrb 0.5 0.5 0.5 0.5>
     outlineSettings {
      bUseBrushTransparency=false
      roundingType=HalfHeightRadius
      width=0
      color {colorUseRule=UseColor_Specified, specifiedColor=<rgba 0 0 0 0>}
      cornerRadii=<0 0 0 0>
     }
     tintColor {colorUseRule=UseColor_Specified, specifiedColor=<rgba 1 1 1 1>}
     uVRegion {bIsValid=false, max=<0 0>, min=<0 0>}
    }
    renderTransformPivot=<0 0.5>
   }
   slot {class=SizeBoxSlot, properties {padding=<ltrb -32 -32 -32 -32>}}
  }
  Overlay_BoostSparks {
   widgetClassPath=&Overlay
   bInherited=true
   bIsVariable=true
   parent=Overlay_FuelBar
   widget=Overlay_BoostSparks
   properties {visibility=Collapsed, renderTransform {angle=90, scale=<1 1>, shear=<0 0>, translation=<0 0>}, renderTransformPivot=<0 0.5>}
   slot {class=OverlaySlot, properties {horizontalAlignment=HAlign_Right, verticalAlignment=VAlign_Center}}
  }
  Image_LargeCone {
   widgetClassPath=&Image
   parent=Overlay_BoostSparks
   widget=Image_LargeCone
   properties {
    visibility=HitTestInvisible
    brush {
     bIsDynamicallyLoaded=false
     drawAs=Image
     imageType=NoImage
     mirroring=NoMirror
     resourceName=None
     resourceObject=/Game/Athena/HUD/Vehicles/Materials/M_UI_SparksGlow_Inst
     tiling=NoTile
     imageSize=<150 219.5743>
     margin=<ltrb 0 0 0 0>
     outlineSettings {
      bUseBrushTransparency=false
      roundingType=HalfHeightRadius
      width=0
      color {colorUseRule=UseColor_Specified, specifiedColor=<rgba 0 0 0 0>}
      cornerRadii=<0 0 0 0>
     }
     tintColor {colorUseRule=UseColor_Specified, specifiedColor=<rgba 1 1 1 1>}
     uVRegion {bIsValid=false, max=<0 0>, min=<0 0>}
    }
    renderTransform {angle=0, scale=<1 1>, shear=<0 0>, translation=<0 44>}
   }
   slot {class=OverlaySlot, properties {horizontalAlignment=HAlign_Center, verticalAlignment=VAlign_Center, padding=<ltrb -200 -200 -200 -200>}}
  }
  Image_Sparks_1 {
   widgetClassPath=&Image
   parent=Overlay_BoostSparks
   widget=Image_Sparks_1
   properties {
    visibility=HitTestInvisible
    brush {
     bIsDynamicallyLoaded=false
     drawAs=Image
     imageType=NoImage
     mirroring=NoMirror
     resourceName=None
     resourceObject=/Game/Athena/UI/InGame/Interrogation/Materials/M_UI_Starburst_Sparks_Inst
     tiling=NoTile
     imageSize=<120 400>
     margin=<ltrb 0 0 0 0>
     outlineSettings {
      bUseBrushTransparency=false
      roundingType=HalfHeightRadius
      width=0
      color {colorUseRule=UseColor_Specified, specifiedColor=<rgba 0 0 0 0>}
      cornerRadii=<0 0 0 0>
     }
     tintColor {colorUseRule=UseColor_Specified, specifiedColor=<rgba 1 0.7183 0.651 1>}
     uVRegion {bIsValid=false, max=<0 0>, min=<0 0>}
    }
   }
   slot {class=OverlaySlot, properties {horizontalAlignment=HAlign_Center, verticalAlignment=VAlign_Center, padding=<ltrb -200 -200 -200 -200>}}
  }
  Image_Sparks {
   widgetClassPath=&Image
   parent=Overlay_BoostSparks
   widget=Image_Sparks
   properties {
    visibility=HitTestInvisible
    brush {
     bIsDynamicallyLoaded=false
     drawAs=Image
     imageType=NoImage
     mirroring=NoMirror
     resourceName=None
     resourceObject=/Game/Athena/UI/InGame/Interrogation/Materials/M_UI_Starburst_Sparks
     tiling=NoTile
     imageSize=<295.3884 400>
     margin=<ltrb 0 0 0 0>
     outlineSettings {
      bUseBrushTransparency=false
      roundingType=HalfHeightRadius
      width=0
      color {colorUseRule=UseColor_Specified, specifiedColor=<rgba 0 0 0 0>}
      cornerRadii=<0 0 0 0>
     }
     tintColor {colorUseRule=UseColor_Specified, specifiedColor=<rgba 1 0.3682 0.0743 1>}
     uVRegion {bIsValid=false, max=<0 0>, min=<0 0>}
    }
   }
   slot {class=OverlaySlot, properties {horizontalAlignment=HAlign_Center, verticalAlignment=VAlign_Center, padding=<ltrb -200 -200 -200 -200>}}
  }
  Widget_FuelGauge {widgetClassPath=/Script/FortniteUI.FortVehicleFuelGauge, bInherited=true, widget=null}
  HorizontalBox_FuelCluster {widgetClassPath=&HorizontalBox, bInherited=true, widget=null}
  Image_Speedometer {widgetClassPath=&Image, bInherited=true, widget=null}
  Image_BoostMeter {widgetClassPath=&Image, bInherited=true, widget=null}
  NumericText_Speed {widgetClassPath=/Script/CommonUI.CommonNumericTextBlock, bInherited=true, widget=null}
  Overlay_Gauge {widgetClassPath=&Overlay, bInherited=true, widget=null}
  Text_ItemName {widgetClassPath=&CommonTextBlock, bInherited=true, widget=null}
  TutorialHighlight {widgetClassPath=&TutorialHighlight}
 ]
}
